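# Unlit simulation material for the water fluid sim's height step, packing water depth, ground height and foam into three channels.
# It semi-Lagrangian advects HeightRT using averaged neighboring VelocityRT samples, subtracts evaporation and adds ForcesRT.
# Static switches select terrain water-body data or a scene-capture texture for ground height, exponentially easing depth back toward rest level.
# Optional foam is advected and faded, then injected where velocity and computed surface slope are high.
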Material /Water/FluidSimulation/Materials/Advect/M_Advect_Height_01
graph {
 properties {bAllowNegativeEmissiveColor=true, pixelDepthOffsetMode=PDOM_Legacy, shadingModel=MSM_Unlit, usedShadingModels=Unlit}
 nodes [
  Constant() {r=-50}
  n1 = TextureSampleParameter2D!(
   UVs=MaterialFunctionCall() {materialFunction=/FortWater/FluidSimulation/Materials/Functions/PointSampleFluidSimUVs}.Result
   Level
   Apply View MipBias
  ) {constMipValue=0, mipValueMode=TMVM_MipLevel, parameterName=HeightRT, texture=&Black}
  n2 = Constant() {r=-65000}
  n3 = MaterialFunctionCall!(Position=MaterialFunctionCall!(UV) {materialFunction=/Water/FluidSimulation/Materials/Functions/Fluidsim_RenderTargetWorldPosition}.Fluidsim WorldPosition) {materialFunction=/Water/Materials/Functions/WaterBodyData}
  n4 = Max(A=@n3.Water Depth) {constB=0}
  n5 = MaterialFunctionCall() {materialFunction=&PointSampleFluidSimUVs}
  n6 = Divide(B=CollectionParameter() {collection=/Water/Materials/MPC/MPC_Water, parameterName=FluidSimResolution})
  n7 = Multiply(A=@n6, B=Constant2Vector() {r=1})
  n8 = Add(
   A=TextureSampleParameter2D!(UVs=@n5.Result, Apply View MipBias) {parameterName=VelocityRT, texture=&Black}.R
   B=TextureSampleParameter2D(UVs=Add(A=@n5.Result, B=@n7)) {parameterName=VelocityRT, texture=&Black}.R
  )
  n9 = Multiply(A=@n6, B=Constant2Vector() {g=1})
  n10 = Add(
   A=TextureSampleParameter2D(UVs=@n5.Result) {parameterName=VelocityRT, texture=&Black}.G
   B=TextureSampleParameter2D(UVs=Add(A=@n5.Result, B=@n9)) {parameterName=VelocityRT, texture=&Black}.G
  )
  n11 = Multiply(
   A=CollectionParameter() {collection=&MPC_Water, parameterName=DeltaTime}
   B=AppendVector(A=Divide(A=@n8), B=Divide(A=@n10))
  )
  n12 = CollectionParameter() {collection=&MPC_Water, parameterName=FluidSimSize}
  n13 = TextureSampleParameter2D!(
   UVs=Subtract(
    A=MaterialFunctionCall() {materialFunction=&PointSampleFluidSimUVs}.Result
    B=Divide(A=@n11, B=@n12)
   )
   Level
   Apply View MipBias
  ) {constMipValue=0, mipValueMode=TMVM_MipLevel, parameterName=HeightRT, texture=&Black}
  n14 = ScalarParameter() {defaultValue=0.05, parameterName=Evaporation}
  n15 = CollectionParameter() {collection=&MPC_Water, parameterName=DeltaTime}
  n16 = Add(
   A=Subtract(A=@n13.R, B=Multiply(A=@n14, B=@n15))
   B=TextureSampleParameter2D(UVs=MaterialFunctionCall() {materialFunction=&PointSampleFluidSimUVs}.Result) {constMipValue=0, mipValueMode=TMVM_MipLevel, parameterName=ForcesRT, texture=&Black}.R
  ) {constB=0}
  MaterialExpressionScalarParameter_3 = ScalarParameter() {defaultValue=0.25, parameterName=Fading Back}
  MaterialExpressionCollectionParameter_7 = CollectionParameter() {collection=&MPC_Water, parameterName=DeltaTime}
  n17 = Custom!(x=@MaterialExpressionScalarParameter_3, t=@MaterialExpressionCollectionParameter_7) {
   code="1.0-exp(-x*t);"
   description=Exp Damping
   outputType=CMOT_Float1
   inputs [
    {inputName=x, input {expression=&self:MaterialExpressionScalarParameter_3}}
    {inputName=t, input {expression=&self:MaterialExpressionCollectionParameter_7}}
   ]
  }
  n18 = StaticSwitchParameter(True=LinearInterpolate(A=If(A=@n1.G, B=@n2, "A > B"=@n16, "A == B"=@n16, "A < B"=@n4) {equalsThreshold=0.5}, B=@n4, Alpha=@n17), False=@n16) {parameterName=Use Terrain System}
  n19 = CollectionParameter() {collection=&MPC_Water, parameterName=WaveCaptureSize}
  n20 = CollectionParameter() {collection=&MPC_Water, parameterName=FluidSimSize}
  n21 = TextureSample(UVs=MaterialFunctionCall(UVs=TextureCoordinate(), Texture Scale=Divide(A=@n19, B=@n20) {desc=Size of Wave Capture vs Sim area}) {materialFunction=&ScaleUVsByCenter}.UVs) {samplerType=SAMPLERTYPE_LinearColor, texture=/Water/FluidSimulation/WaveFoam/WaterDistanceField}
  n22 = Max(A=Subtract(A=@n3.Water Z Height, B=@n21.G) {constA=0}) {constB=0}
  n23 = If(A=@n1.G, B=Constant() {r=-65000}, "A > B"=@n18, "A == B"=@n18, "A < B"=@n22) {equalsThreshold=0.5}
  MaterialExpressionScalarParameter_7 = ScalarParameter() {defaultValue=0.25, parameterName=Fading Back}
  MaterialExpressionCollectionParameter_6 = CollectionParameter() {collection=&MPC_Water, parameterName=DeltaTime}
  n24 = Custom(x=@MaterialExpressionScalarParameter_7, t=@MaterialExpressionCollectionParameter_6) {
   code="1.0-exp(-x*t);"
   description=Exp Damping
   outputType=CMOT_Float1
   inputs [
    {inputName=x, input {expression=&self:MaterialExpressionScalarParameter_7}}
    {inputName=t, input {expression=&self:MaterialExpressionCollectionParameter_6}}
   ]
  }
  n25 = LinearInterpolate(
   A=@n3.Terrain Z Height
   B=Multiply(A=@n3.Water Depth) {constB=-1, desc=Water as Flat Plane Mode}
   Alpha=ScalarParameter() {parameterName=TerrainUseWaterDepthOnly}
  )
  n26 = Constant()
  n27 = Multiply(
   A=ScalarParameter() {defaultValue=2, parameterName=Foam Fading}
   B=CollectionParameter() {collection=&MPC_Water, parameterName=DeltaTime}
  )
  n28 = Multiply(A=@n13.B, B=Saturate(None=OneMinus(None=@n27)))
  n29 = TextureSampleParameter2D!(UVs, Apply View MipBias) {parameterName=VelocityRT, texture=&Black}
  n30 = Divide(
   A=MaterialFunctionCall(Vector 2=ComponentMask(None=@n29.RGB) {g=true, r=true}) {materialFunction=/Engine/Functions/Engine_MaterialFunctions02/Utility/VectorLength}.V2 Length
   B=ScalarParameter() {defaultValue=128, parameterName=Velocity for Foam}
  )
  MaterialExpressionTextureObjectParameter_1 = TextureObjectParameter() {
   parameterName=HeightRT
   samplerType=SAMPLERTYPE_LinearColor
   texture=/Engine/Functions/Engine_MaterialFunctions02/PivotPainter2/Black_1x1_EXR_Texture
  }
  MaterialExpressionCollectionParameter_8 = CollectionParameter() {collection=&MPC_Water, parameterName=FluidSimResolution}
  MaterialExpressionTextureCoordinate_1 = TextureCoordinate()
  MaterialExpressionCollectionParameter_9 = CollectionParameter() {collection=&MPC_Water, parameterName=FluidTexelSize}
  MaterialExpressionScalarParameter_2 = ScalarParameter() {defaultValue=5, parameterName=ZScale}
  n31 = Custom!(
   HeightTex=@MaterialExpressionTextureObjectParameter_1
   TextureSize=@MaterialExpressionCollectionParameter_8
   UV=@MaterialExpressionTextureCoordinate_1
   GridSize=@MaterialExpressionCollectionParameter_9
   HeightScale=@MaterialExpressionScalarParameter_2
  ) {
   code="//HeightTex: input Texture Object containing heights.R and water.B\r\n//TextureSize: input for size\r\n//GridSize: input float grid world size\r\n//HeightScale: input for height range scale\r\n\r\nfloat2 offsets[4] = {float2(0,1), float2(0,-1), float2(1,0), float2(-1,0)};\r\n\r\nfloat2 texelsize = 1 / TextureSize;\r\n\r\nUV = (floor(UV * TextureSize)/TextureSize) + (texelsize * 0.5);\r\n\r\nfloat2 PosYSample = HeightTex.SampleLevel(HeightTexSampler, UV + offsets[0] * texelsize, 0).rg;\r\nfloat2 NegYSample = HeightTex.SampleLevel(HeightTexSampler, UV + offsets[1] * texelsize, 0).rg;\r\nfloat2 PosXSample = HeightTex.SampleLevel(HeightTexSampler, UV + offsets[2] * texelsize, 0).rg;\r\nfloat2 NegXSample = HeightTex.SampleLevel(HeightTexSampler, UV + offsets[3] * texelsize, 0).rg;\r\n\r\nPosYSample.x +=  PosYSample.y;\r\nNegYSample.x +=  NegYSample.y;\r\nPosXSample.x +=  PosXSample.y;\r\nNegXSample.x +=  NegXSample.y;\r\n\r\nfloat3 tangent =  float3( 2*GridSize, 0, ((PosXSample.x) - (NegXSample.x))*HeightScale);\r\nfloat3 binormal = float3( 0, 2*GridSize, ((PosYSample.x) - (NegYSample.x))*HeightScale);\r\ntangent = normalize(tangent);\r\nbinormal = normalize(binormal);\r\n\r\nfloat3 normal = normalize(cross(tangent, binormal));\r\n\r\nreturn normal;"
   description=Compute Normal
   inputs [
    {
     inputName=HeightTex
     input {expression=&self:MaterialExpressionTextureObjectParameter_1}
    }
    {inputName=TextureSize, input {expression=&self:MaterialExpressionCollectionParameter_8}}
    {inputName=UV, input {expression=&self:MaterialExpressionTextureCoordinate_1}}
    {inputName=GridSize, input {expression=&self:MaterialExpressionCollectionParameter_9}}
    {inputName=HeightScale, input {expression=&self:MaterialExpressionScalarParameter_2}}
   ]
  }
  n32 = Subtract(
   A=MaterialFunctionCall(Vector 2=ComponentMask(None=@n31) {g=true, r=true}) {materialFunction=&VectorLength}.V2 Length
   B=ScalarParameter() {defaultValue=0.5, parameterName=Min Z for Foam}
  )
  n33 = ScalarParameter() {defaultValue=0.05, parameterName=Foam Z Len}
  n34 = Multiply(
   A=Multiply(A=Saturate(None=@n30), B=ScalarParameter() {defaultValue=0.5, parameterName=Foam Injection})
   B=Saturate(None=Divide(A=@n32, B=@n33))
  )
  MaterialFunctionCall(
   X=StaticSwitchParameter(True=LinearInterpolate(A=@n23, B=@n22, Alpha=@n24), False=@n18) {parameterName=Use Scene Capture}
   Y=StaticSwitchParameter(True=@n21.G, False=StaticSwitchParameter(True=@n25, False=@n26) {parameterName=Use Terrain System}) {parameterName=Use Scene Capture}
   Z=StaticSwitchParameter(True=Add(A=@n28, B=@n34), False=Constant()) {defaultValue=true, parameterName=Enable Foam}
  ) {materialFunction=&MakeFloat3}
 ]
}
