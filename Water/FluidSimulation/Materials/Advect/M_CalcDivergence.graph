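# Unlit fluid simulation material that updates water height from the divergence of the velocity field while passing terrain height through.
# It computes divergence from staggered wall velocities in VelocityRT (X on left/right, Y on bottom/top) over FluidTexelSize from MPC_Water.
# Water height from HeightRT red is divided by one plus divergence times DeltaTime, an implicit-style step, then kept non-negative and clamped.
# HeightRT green carries terrain Z unchanged into the other output channels.
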Material /Water/FluidSimulation/Materials/Advect/M_CalcDivergence
graph {
 properties {bAllowNegativeEmissiveColor=true, pixelDepthOffsetMode=PDOM_Legacy, shadingModel=MSM_Unlit, usedShadingModels=Unlit}
 nodes [
  n1 = TextureSampleParameter2D!(
   UVs=MaterialFunctionCall() {materialFunction=/FortWater/FluidSimulation/Materials/Functions/PointSampleFluidSimUVs}.Result
   Level=Constant()
   Apply View MipBias
  ) {constMipValue=0, mipValueMode=TMVM_MipLevel, parameterName=HeightRT, texture=&Black}
  n2 = MaterialFunctionCall() {materialFunction=&PointSampleFluidSimUVs}
  n3 = Multiply(A=Divide(B=CollectionParameter() {collection=/Water/Materials/MPC/MPC_Water, parameterName=FluidSimResolution}))
  n4 = Constant2Vector() {r=1}
  n5 = TextureSampleParameter2D(UVs=Add(A=@n2.Result, B=Multiply(A=@n3, B=@n4))) {constMipValue=0, desc=Right Wall, mipValueMode=TMVM_MipLevel, parameterName=VelocityRT, texture=&Black}
  n6 = TextureSampleParameter2D(UVs=@n2.Result) {constMipValue=0, desc=Left Wall, mipValueMode=TMVM_MipLevel, parameterName=VelocityRT, texture=&Black}
  n7 = MaterialFunctionCall() {materialFunction=&PointSampleFluidSimUVs}
  n8 = Constant2Vector() {g=1}
  n9 = TextureSampleParameter2D(UVs=Add(A=@n7.Result, B=Multiply(A=@n3, B=@n8))) {constMipValue=0, desc=Top Wall, mipValueMode=TMVM_MipLevel, parameterName=VelocityRT, texture=&Black}
  n10 = TextureSampleParameter2D(UVs=@n7.Result) {constMipValue=0, desc=Bottom Wall, mipValueMode=TMVM_MipLevel, parameterName=VelocityRT, texture=&Black}
  n11 = Divide(
   A=Add(A=Subtract(A=@n5.R, B=@n6.R), B=Subtract(A=@n9.G, B=@n10.G))
   B=CollectionParameter() {collection=&MPC_Water, parameterName=FluidTexelSize}
  )
  n12 = CollectionParameter() {collection=&MPC_Water, parameterName=DeltaTime}
  n13 = Divide(A=@n1.R, B=Add(B=Multiply(A=@n11, B=@n12)) {constA=1})
  MaterialFunctionCall(X=Clamp(None=Max(A=@n13) {constB=0, desc=Updated Water}) {maxDefault=32767, minDefault=-32768}, Y=@n1.G, Z=@n1.G) {desc=Pass through Terrain Z, materialFunction=&MakeFloat3}
 ]
}
