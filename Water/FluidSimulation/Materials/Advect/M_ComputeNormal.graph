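# Unlit fluid-simulation material that packs water normal XY and a height offset from the simulation render target, allowing negative emissive values.
# A Custom HLSL node derives normals by central differences of summed R and G channels plus exponential foam height from B.
# Static switches choose height relative to WaterBodyData (scene capture or terrain, sinking empty areas) or a default depth.
# A box mask fades edges; side branches compute steep-cliff masking and neighbour dilation.
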Material /Water/FluidSimulation/Materials/Advect/M_ComputeNormal
graph {
 properties {
  bAllowNegativeEmissiveColor=true
  blendMode=BLEND_AlphaComposite
  pixelDepthOffsetMode=PDOM_Legacy
  shadingModel=MSM_Unlit
  usedShadingModels=Unlit
 }
 nodes [
  n1 = TextureSampleParameter2D() {
   desc=Fluidsim
   parameterName=RT
   samplerType=SAMPLERTYPE_LinearColor
   texture=/Engine/Functions/Engine_MaterialFunctions02/PivotPainter2/Black_1x1_EXR_Texture
  }
  n2 = MaterialFunctionCall!(Position=MaterialFunctionCall!(UV) {materialFunction=/Water/FluidSimulation/Materials/Functions/Fluidsim_RenderTargetWorldPosition}.Fluidsim WorldPosition) {materialFunction=/Water/Materials/Functions/WaterBodyData}
  Add(A=Subtract(A=@n1.G, B=@n2.Water Z Height), B=ScalarParameter() {defaultValue=-6, parameterName=Empty Water Depth Bias})
  CollectionParameter() {collection=/Water/Materials/MPC/MPC_Water, parameterName=FluidTexelSize}
  Constant3Vector() {constant=<rgba 0 0.2 0.3 0>}
  Constant()
  n3 = TextureSampleParameter2D() {
   desc=Fluidsim
   parameterName=RT
   samplerType=SAMPLERTYPE_LinearColor
   texture=&Black_1x1_EXR_Texture
  }
  Divide(A=@n3.R, B=ScalarParameter() {defaultValue=1000, parameterName=MaxEncodedDepth})
  MaterialExpressionTextureObjectParameter_1 = TextureObjectParameter() {
   parameterName=RT
   samplerType=SAMPLERTYPE_LinearColor
   texture=&Black_1x1_EXR_Texture
  }
  MaterialExpressionCollectionParameter_4 = CollectionParameter() {collection=&MPC_Water, parameterName=FluidSimResolution}
  MaterialExpressionTextureCoordinate_1 = TextureCoordinate()
  MaterialExpressionCollectionParameter_5 = CollectionParameter() {collection=&MPC_Water, parameterName=FluidTexelSize}
  MaterialExpressionScalarParameter_0 = ScalarParameter() {defaultValue=1, parameterName=ZScale}
  MaterialExpressionScalarParameter_9 = ScalarParameter() {defaultValue=0.01, parameterName=Foam Affects Normal}
  MaterialExpressionScalarParameter_10 = ScalarParameter() {defaultValue=16, parameterName=FHS}
  n4 = Custom!(
   HeightTex=@MaterialExpressionTextureObjectParameter_1
   TextureSize=@MaterialExpressionCollectionParameter_4
   UV=@MaterialExpressionTextureCoordinate_1
   GridSize=@MaterialExpressionCollectionParameter_5
   HeightScale=@MaterialExpressionScalarParameter_0
   Foam=@MaterialExpressionScalarParameter_9
   FoamHeightScale=@MaterialExpressionScalarParameter_10
  ) {
   code="//HeightTex: input Texture Object containing heights.R and water.B\r\n//TextureSize: input for size\r\n//GridSize: input float grid world size\r\n//HeightScale: input for height range scale\r\n\r\nfloat2 offsets[4] = {float2(0,1), float2(0,-1), float2(1,0), float2(-1,0)};\r\n\r\nfloat2 texelsize = 1 / TextureSize;\r\n\r\nUV = (floor(UV * TextureSize)/TextureSize) + (texelsize * 0.5);\r\n\r\nfloat3 PosYSample = HeightTex.SampleLevel(HeightTexSampler, UV + offsets[0] * texelsize, 0).rgb;\r\nfloat3 NegYSample = HeightTex.SampleLevel(HeightTexSampler, UV + offsets[1] * texelsize, 0).rgb;\r\nfloat3 PosXSample = HeightTex.SampleLevel(HeightTexSampler, UV + offsets[2] * texelsize, 0).rgb;\r\nfloat3 NegXSample = HeightTex.SampleLevel(HeightTexSampler, UV + offsets[3] * texelsize, 0).rgb;\r\n\r\nPosYSample.x +=  PosYSample.y;\r\nNegYSample.x +=  NegYSample.y;\r\nPosXSample.x +=  PosXSample.y;\r\nNegXSample.x +=  NegXSample.y;\r\n\r\nPosYSample.x +=  (1- exp(-PosYSample.z * Foam)) * FoamHeightScale*GridSize;\r\nNegYSample.x +=  (1-exp(-NegYSample.z * Foam)) * FoamHeightScale*GridSize;\r\nPosXSample.x +=  (1-exp(-PosXSample.z * Foam)) * FoamHeightScale*GridSize;\r\nNegXSample.x +=  (1-exp(-NegXSample.z * Foam)) * FoamHeightScale*GridSize;\r\n\r\nfloat3 tangent =  float3( 2*GridSize, 0, ((PosXSample.x) - (NegXSample.x))*HeightScale);\r\nfloat3 binormal = float3( 0, 2*GridSize, ((PosYSample.x) - (NegYSample.x))*HeightScale);\r\ntangent = normalize(tangent);\r\nbinormal = normalize(binormal);\r\n\r\nfloat3 normal = normalize(cross(tangent, binormal));\r\n\r\nreturn normal;"
   description=Compute Normal
   inputs [
    {inputName=HeightTex, input {expression=&self:MaterialExpressionTextureObjectParameter_1}}
    {inputName=TextureSize, input {expression=&self:MaterialExpressionCollectionParameter_4}}
    {inputName=UV, input {expression=&self:MaterialExpressionTextureCoordinate_1}}
    {inputName=GridSize, input {expression=&self:MaterialExpressionCollectionParameter_5}}
    {inputName=HeightScale, input {expression=&self:MaterialExpressionScalarParameter_0}}
    {inputName=Foam, input {expression=&self:MaterialExpressionScalarParameter_9}}
    {inputName=FoamHeightScale, input {expression=&self:MaterialExpressionScalarParameter_10}}
   ]
  }
  n5 = ComponentMask(None=@n4) {g=true, r=true}
  n6 = ScalarParameter() {defaultValue=1, parameterName=Height for Full Normals}
  LinearInterpolate(B=@n5, Alpha=Saturate(None=Divide(A=@n1.R, B=@n6)))
  n7 = Add(A=@n1.R, B=@n1.R)
  n8 = LinearInterpolate(
   A=Constant()
   B=Add(A=Subtract(A=@n7, B=@n2.Water Z Height), B=ScalarParameter() {parameterName=Water Z Bias})
   Alpha=Saturate(None=Ceil(None=@n1.R))
  )
  n9 = MaterialFunctionCall(Position=MaterialFunctionCall() {materialFunction=&Fluidsim_RenderTargetWorldPosition}.Fluidsim WorldPosition) {materialFunction=&WaterBodyData}
  n10 = Max(A=Add(A=@n9.Water Depth, B=ScalarParameter() {defaultValue=256, parameterName=Water Empty Bias})) {constB=0}
  n11 = Ceil(None=@n9.Water Depth)
  n12 = TextureSampleParameter2D() {
   desc=Fluidsim
   parameterName=RT
   samplerType=SAMPLERTYPE_LinearColor
   texture=&Black_1x1_EXR_Texture
  }
  n13 = Add(A=@n12.R, B=@n9.Terrain Z Height)
  n14 = LinearInterpolate(
   A=LinearInterpolate(B=Multiply(A=@n10) {constB=-1}, Alpha=Saturate(None=@n11)) {desc=Unless the water was already Empy there}
   B=Add(A=Subtract(A=@n13, B=@n9.Water Z Height), B=ScalarParameter() {defaultValue=5, parameterName=Water Terrain Mode Z Bias})
   Alpha=Saturate(None=Ceil(None=@n12.R)) {desc=Mask for where there is Fluid}
  )
  n15 = TextureSampleParameter2D() {
   desc=Fluidsim
   parameterName=RT
   samplerType=SAMPLERTYPE_LinearColor
   texture=&Black_1x1_EXR_Texture
  }
  n16 = StaticSwitchParameter(
   True=StaticSwitchParameter(True=@n14, False=@n14) {parameterName=Use Terrain DepthOnly}
   False=Subtract(A=@n15.R, B=ScalarParameter() {parameterName=DefaultWaterDepth})
  ) {parameterName=Use Terrain System}
  n17 = ScalarParameter() {defaultValue=0.2, parameterName=EdgeFalloff}
  n18 = Multiply(A=@n17) {constB=2}
  n19 = MaterialFunctionCall!(
   Alpha=MaterialFunctionCall!(A=TextureCoordinate(), B=Constant() {r=0.5}, Bounds=OneMinus(None=@n18), Edge Falloff=@n17) {materialFunction=/Engine/Functions/Engine_MaterialFunctions01/Gradient/BoxMask-2D}.Result
   Offset=ScalarParameter() {defaultValue=0.02, parameterName=AlphaOffset}
  ) {materialFunction=/Engine/Functions/Engine_MaterialFunctions01/Opacity/AlphaOffset}
  Multiply(A=AppendVector(A=@n5, B=StaticSwitchParameter(True=@n8, False=@n16) {parameterName=Use Scene Capture}), B=@n19.Result)
  n20 = Subtract(A=ComponentMask(None=@n4) {b=true}, B=ScalarParameter() {defaultValue=0.3, parameterName=Min Z Normal Mask})
  Multiply(A=Saturate(None=Ceil(None=@n20))) {desc="Prevent sheer water cliffs, hide em"}
  OneMinus(None=Saturate(None=Multiply(A=@n19.Result, B=@n1.B)))
  MaterialExpressionTextureObjectParameter_0 = TextureObjectParameter() {
   parameterName=RT
   samplerType=SAMPLERTYPE_LinearColor
   texture=&Black_1x1_EXR_Texture
  }
  MaterialExpressionCollectionParameter_0 = CollectionParameter() {collection=&MPC_Water, parameterName=FluidSimResolution}
  MaterialExpressionTextureCoordinate_2 = TextureCoordinate()
  MaterialExpressionScalarParameter_8 = ScalarParameter() {defaultValue=4, parameterName=Dilate Scale}
  n21 = Custom!(
   HeightTex=@MaterialExpressionTextureObjectParameter_0
   TextureSize=@MaterialExpressionCollectionParameter_0
   UV=@MaterialExpressionTextureCoordinate_2
   OffsetScale=@MaterialExpressionScalarParameter_8
  ) {
   code=" //HeightTex: input Texture Object containing heights.R and water.B\r\n//TextureSize: input for size\r\n//GridSize: input float grid world size\r\n//HeightScale: input for height range scale\r\n\r\nfloat2 offsets[4] = {float2(0,1), float2(0,-1), float2(1,0), float2(-1,0)};\r\n\r\nfloat2 texelsize = 1 / TextureSize;\r\ntexelsize *= OffsetScale;\r\n\r\nUV = (floor(UV * TextureSize)/TextureSize) + (texelsize * 0.5);\r\n\r\nfloat PosYSample = HeightTex.SampleLevel(HeightTexSampler, UV + offsets[0] * texelsize, 0).r;\r\nfloat NegYSample = HeightTex.SampleLevel(HeightTexSampler, UV + offsets[1] * texelsize, 0).r;\r\nfloat PosXSample = HeightTex.SampleLevel(HeightTexSampler, UV + offsets[2] * texelsize, 0).r;\r\nfloat NegXSample = HeightTex.SampleLevel(HeightTexSampler, UV + offsets[3] * texelsize, 0).r;\r\n\r\nfloat maxval = max(PosXSample, max(NegXSample, max(PosYSample, NegYSample)));\r\nfloat mask =saturate(ceil( PosXSample + NegXSample + PosYSample + NegYSample));\r\n\r\n//returns 1 if any neighbor had fluid\r\nreturn float2(mask, maxval);"
   description=Neighbor Check
   outputType=CMOT_Float2
   inputs [
    {inputName=HeightTex, input {expression=&self:MaterialExpressionTextureObjectParameter_0}}
    {inputName=TextureSize, input {expression=&self:MaterialExpressionCollectionParameter_0}}
    {inputName=UV, input {expression=&self:MaterialExpressionTextureCoordinate_2}}
    {inputName=OffsetScale, input {expression=&self:MaterialExpressionScalarParameter_8}}
   ]
  }
  Subtract(
   A=Add(
    A=@n1.G
    B=MaterialFunctionCall(Float2=@n21) {materialFunction=&BreakOutFloat2Components}.G
   )
   B=@n2.Water Z Height
  )
 ]
}
