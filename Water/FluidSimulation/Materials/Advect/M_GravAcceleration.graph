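# Unlit simulation material that computes the gravity-driven velocity update for the water fluid simulation.
# A Custom HLSL node compares water-plus-terrain elevation from HeightRT with the -X and -Y neighbors, turning slopes into acceleration added to VelocityRT.
# EdgeLength fades acceleration where uphill water is shallow, preventing flow from dry cells.
# MPC_Water supplies timestep, grid scale and exponential damping; the wind term is wired to a zero constant.
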Material /Water/FluidSimulation/Materials/Advect/M_GravAcceleration
graph {
 properties {bAllowNegativeEmissiveColor=true, pixelDepthOffsetMode=PDOM_Legacy, shadingModel=MSM_Unlit, usedShadingModels=Unlit}
 nodes [
  MaterialExpressionTextureObjectParameter_0 = TextureObjectParameter() {parameterName=HeightRT, texture=&Black}
  MaterialExpressionTextureSampleParameter2D_7 = TextureSampleParameter2D!(UVs, Level=Constant(), Apply View MipBias) {mipValueMode=TMVM_MipLevel, parameterName=VelocityRT, texture=&Black}
  MaterialExpressionCollectionParameter_2 = CollectionParameter() {collection=/Water/Materials/MPC/MPC_Water, parameterName=DeltaTime}
  n1 = Divide(B=CollectionParameter() {collection=&MPC_Water, parameterName=FluidSimResolution})
  MaterialExpressionAppendVector_1 = AppendVector(A=@n1, B=@n1)
  n2 = CollectionParameter() {collection=&MPC_Water, parameterName=FluidTexelSize}
  MaterialExpressionAppendVector_0 = AppendVector(A=@n2, B=@n2)
  MaterialExpressionMaterialFunctionCall_9 = MaterialFunctionCall() {materialFunction=/FortWater/FluidSimulation/Materials/Functions/PointSampleFluidSimUVs}
  MaterialExpressionCollectionParameter_3 = CollectionParameter() {collection=&MPC_Water, parameterName=FluidDamping}
  MaterialExpressionConstant_1 = Constant()
  MaterialExpressionScalarParameter_0 = ScalarParameter() {defaultValue=0.1, parameterName=EdgeLength}
  Custom!(
   HeightRT=@MaterialExpressionTextureObjectParameter_0
   Velocity=@MaterialExpressionTextureSampleParameter2D_7.RGB
   deltaT=@MaterialExpressionCollectionParameter_2
   TexelSize=@MaterialExpressionAppendVector_1
   GridSize=@MaterialExpressionAppendVector_0
   inUV=@MaterialExpressionMaterialFunctionCall_9.Result
   Damping=@MaterialExpressionCollectionParameter_3
   Wind=@MaterialExpressionConstant_1
   EdgeLength=@MaterialExpressionScalarParameter_0
  ) {
   code="//HeightData: input v4, R=Terrain Height, G=Fluid Height(bottom), B=FluidVolulume, A=Sediment fraction\r\n//Velocity: input \r\n//NormalRT: input \r\n//HeightRT: input\r\n//deltaT: input\r\n//input TextureSize\r\n//input TexelSize\r\n//input GridSize\r\n\r\nconst float grav = 980;\r\n\r\n//R = water depth,  G = terrain Z elevation\r\nfloat2 startZ = HeightRT.SampleLevel(HeightRTSampler, inUV, 0).rg;\r\nfloat2 offsetX = HeightRT.SampleLevel(HeightRTSampler, inUV - float2(TexelSize.x, 0), 0).rg;\r\nfloat2 offsetY = HeightRT.SampleLevel(HeightRTSampler, inUV - float2(0, TexelSize.y), 0).rg;\r\n\r\nfloat3 outvel = 0;\r\n\r\n\r\nfloat elevation_0 = startZ.x + startZ.y;\r\nfloat elevation_x = offsetX.x + offsetX.y;\r\nfloat elevation_y = offsetY.x + offsetY.y;\r\n\r\noutvel.x = grav * ((elevation_x-elevation_0) / GridSize) * deltaT;\r\noutvel.y = grav * ((elevation_y-elevation_0) / GridSize) * deltaT;\r\n\r\nfloat2 top = startZ;\r\nfloat2 second = offsetX;\r\n\t\t\r\n//is startZ actually lower than offsetX?\r\nif(top.x + top.y < second.x + second.y) \r\n{\r\n\ttop = offsetX;\r\n\tsecond = startZ;\r\n}\r\n\t\t\r\n//Avoid applying gravity if there is no water uphill\r\nif( top.x <= EdgeLength)\r\n{\r\n\tfloat mask = saturate(top.x /  EdgeLength);\r\n\toutvel.x *= mask;\r\n}\r\n\r\ntop = startZ;\r\nsecond = offsetY;\r\n\t\t\r\n//is startZ actually lower than offsetX?\r\nif(top.x + top.y < second.x + second.y) \r\n{\r\n\ttop = offsetY;\r\n\tsecond = startZ;\r\n}\r\n\r\n//Avoid applying gravity if there is no water uphill\r\nif( top.x <= EdgeLength )\r\n{\r\n\tfloat mask = saturate(top.x /  EdgeLength);\r\n\toutvel.y *= mask;\r\n}\r\n\r\n//Gravity acceleration added to initial velocity\r\noutvel += Velocity;\r\n\r\n//Apply wind, if any\r\noutvel += Wind * startZ.y * deltaT;\r\n\r\n//Damping\r\noutvel *= exp(-Damping*deltaT);\r\n\r\n\r\nreturn outvel;\r\n\r\n"
   description=Velocity Acceleration
   inputs [
    {inputName=HeightRT, input {expression=&self:MaterialExpressionTextureObjectParameter_0}}
    {
     inputName=Velocity
     input {
      expression=&self:MaterialExpressionTextureSampleParameter2D_7
      mask=1
      maskB=1
      maskG=1
      maskR=1
     }
    }
    {inputName=deltaT, input {expression=&self:MaterialExpressionCollectionParameter_2}}
    {inputName=TexelSize, input {expression=&self:MaterialExpressionAppendVector_1}}
    {inputName=GridSize, input {expression=&self:MaterialExpressionAppendVector_0}}
    {inputName=inUV, input {expression=&self:MaterialExpressionMaterialFunctionCall_9}}
    {inputName=Damping, input {expression=&self:MaterialExpressionCollectionParameter_3}}
    {inputName=Wind, input {expression=&self:MaterialExpressionConstant_1}}
    {inputName=EdgeLength, input {expression=&self:MaterialExpressionScalarParameter_0}}
   ]
  }
 ]
}
